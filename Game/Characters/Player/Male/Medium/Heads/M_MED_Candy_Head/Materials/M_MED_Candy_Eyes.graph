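# Masked eye material for the Candy player head, swapping vertex-colored eye and brow shapes to match facial poses.
# Pose scalars pass thresholded If chains into VertexColorMask on red and green channels, with a dizzy-face branch and a static switch for alternate eyebrow logic.
# The mask drives World Position Offset, collapsing hidden elements onto the object position.
# A texture supplies color, opacity and optional emissive, then passes through shared tech-art override and character opacity functions.
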Material /Game/Characters/Player/Male/Medium/Heads/M_MED_Candy_Head/Materials/M_MED_Candy_Eyes
graph {
 properties {
  bUseMaterialAttributes=true
  bUsedWithMeshDeformer=true
  bUsedWithSkeletalMesh=true
  blendMode=BLEND_Masked
  pixelDepthOffsetMode=PDOM_Legacy
  usedShadingModels=Default Lit
 }
 nodes [
  n1 = VertexColor()
  n2 = ChannelMaskParameter(None=@n1) {group=Banana, parameterName=Banana_Former Red}
  n3 = ChannelMaskParameter(None=@n1) {group=Banana, maskChannel=Green, parameterName=Banana_Former Green, defaultValue=<rgba 0 1 0 0>}
  Add(A=@n2.R, B=@n3.R)
  Add()
  Constant()
  n4 = VertexColor()
  n5 = ScalarParameter() {group=Banana_Poses, parameterName=R_brow_up_pose}
  n6 = ScalarParameter() {defaultValue=0.1, group=Banana, parameterName=BananaEyeThreshold}
  n7 = ScalarParameter() {defaultValue=0.4, group=Banana, parameterName="BananaBrowUpValue(G)"}
  n8 = Constant() {r=1}
  n9 = If(A=@n5, B=@n6, "A > B"=@n7, "A == B"=@n8, "A < B"=@n8)
  n10 = ScalarParameter() {group=Banana_Poses, parameterName=C_glabella_down_pose}
  n11 = ScalarParameter() {defaultValue=0.2, group=Banana, parameterName="GlabellaDownValue(G)"}
  n12 = If(A=@n10, B=@n6, "A > B"=@n11, "A == B"=@n8, "A < B"=@n8)
  n13 = ScalarParameter() {group=Banana_Poses, parameterName=C_glabella_up_pose}
  n14 = ScalarParameter() {defaultValue=0.4, group=Banana, parameterName="GlabellaUpValue(G)"}
  n15 = MaterialFunctionCall!(
   Hide Element Toggle=Constant() {r=1}
   VertexColorChannel=@n4.G
   Vertex Value to hide=Min(A=Min(A=@n9, B=@n12), B=If(A=@n13, B=@n6, "A > B"=@n14, "A == B"=@n8, "A < B"=@n8))
  ) {materialFunction=/Game/Packages/Fortress_SharedMaterials/Base_Material_Functions/VertexColorMask}
  n16 = If(A=@n4.G, B=Constant() {r=0.95}, "A > B"=Constant() {r=1}, "A == B"=@n15.Result, "A < B"=@n15.Result)
  n17 = MaterialFunctionCall(
   Hide Element Toggle=Constant() {r=1}
   VertexColorChannel=@n4.R
   Vertex Value to hide=ScalarParameter() {defaultValue=0.3, group=Banana, parameterName="BananaHappyPoseValue(R)"}
  ) {materialFunction=&VertexColorMask}
  n18 = If(
   A=ScalarParameter() {group=Banana_Poses, parameterName=Squint_inner_pose}
   B=@n6
   "A > B"=MaterialFunctionCall(
    Hide Element Toggle=Constant() {r=1}
    VertexColorChannel=@n4.R
    Vertex Value to hide=ScalarParameter() {defaultValue=0.5, group=Banana, parameterName="BananaSquintPoseValue(R)"}
   ) {materialFunction=&VertexColorMask}.Result
   "A == B"=@n17.Result
   "A < B"=@n17.Result
  )
  n19 = If(
   A=ScalarParameter() {group=Banana_Poses, parameterName=R_blink_pose}
   B=@n6
   "A > B"=MaterialFunctionCall(
    Hide Element Toggle=Constant() {r=1}
    VertexColorChannel=@n4.R
    Vertex Value to hide=ScalarParameter() {defaultValue=0.6, group=Banana, parameterName="BananaBlinkPoseValue(R)"}
   ) {materialFunction=&VertexColorMask}.Result
   "A == B"=@n18
   "A < B"=@n18
  )
  n20 = Constant()
  If(A=Divide(A=Add(A=@n4.R, B=@n4.R)), B=Constant() {r=1}, "A > B"=Add(A=@n16, B=@n19), "A == B"=@n20, "A < B"=@n20)
  n21 = TextureSampleParameter2D() {parameterName=Eye_Texture, texture=/Game/Characters/Player/Male/Medium/Heads/M_MED_Candy_Head/Textures/CandyEyes}
  n22 = Multiply(A=Multiply(A=@n21.RGB, B=ScalarParameter() {defaultValue=1, parameterName=EmissiveStrength}), B=@n21.A)
  n23 = ScalarParameter() {parameterName=UseEmissveEye}
  n24 = Multiply(A=Subtract(A=WorldPosition().XYZ, B=ObjectPositionWS())) {constB=-1}
  n25 = MaterialFunctionCall(
   Hide Element Toggle=Constant() {r=1}
   VertexColorChannel=@n2.R
   Vertex Value to hide=ScalarParameter() {defaultValue=0.3, group=Banana, parameterName="BananaHappyPoseValue(R)"}
  ) {materialFunction=&VertexColorMask}
  n26 = If(
   A=ScalarParameter() {group=Banana_Poses, parameterName=R_squint_inner_pose}
   B=ScalarParameter() {defaultValue=0.375, group=Banana, parameterName=Threshold_squint_inner_pose}
   "A > B"=MaterialFunctionCall(
    Hide Element Toggle=Constant() {r=1}
    VertexColorChannel=@n2.R
    Vertex Value to hide=ScalarParameter() {defaultValue=0.5, group=Banana, parameterName="BananaSquintPoseValue(R)"}
   ) {materialFunction=&VertexColorMask}.Result
   "A == B"=@n25.Result
   "A < B"=@n25.Result
  )
  n27 = If(
   A=@n5
   B=ScalarParameter() {defaultValue=0.1, group=Banana, parameterName=Threshold_R_brow_up_pose}
   "A > B"=MaterialFunctionCall(Hide Element Toggle=Constant() {r=1}, VertexColorChannel=@n3.R, Vertex Value to hide=@n7) {materialFunction=&VertexColorMask}.Result
   "A == B"=@n26
   "A < B"=@n26
  )
  n28 = If(
   A=@n10
   B=ScalarParameter() {defaultValue=0.1, group=Banana, parameterName=Threshold_Glabella_down}
   "A > B"=MaterialFunctionCall(Hide Element Toggle=Constant() {r=1}, VertexColorChannel=@n3.R, Vertex Value to hide=@n11) {materialFunction=&VertexColorMask}.Result
   "A == B"=@n27
   "A < B"=@n27
  )
  n29 = If(
   A=@n13
   B=ScalarParameter() {defaultValue=0.1, group=Banana, parameterName=Threshold_Glabella_up}
   "A > B"=MaterialFunctionCall(Hide Element Toggle=Constant() {r=1}, VertexColorChannel=@n3.R, Vertex Value to hide=Constant() {r=0.1}) {materialFunction=&VertexColorMask}.Result
   "A == B"=@n28
   "A < B"=@n28
  )
  n30 = If(
   A=ScalarParameter() {group=Banana_Poses, parameterName=R_blink_pose}
   B=ScalarParameter() {defaultValue=0.1, group=Banana, parameterName=Threshold_R_blink_pose}
   "A > B"=MaterialFunctionCall(
    Hide Element Toggle=Constant() {r=1}
    VertexColorChannel=@n2.R
    Vertex Value to hide=ScalarParameter() {defaultValue=0.6, group=Banana, parameterName="BananaBlinkPoseValue(R)"}
   ) {materialFunction=&VertexColorMask}.Result
   "A == B"=@n29
   "A < B"=@n29
  )
  n31 = OneMinus(None=If(A=Max(B=@n3.R), B=Constant() {r=0.95}, "A > B"=Constant() {r=1}, "A == B"=@n30, "A < B"=@n30))
  n32 = Saturate(None=ScalarParameter() {parameterName=PeelyBlenderFace})
  n33 = Constant() {r=1}
  n34 = ScalarParameter() {group=Banana, parameterName="DizzyFaceValue(R)"}
  n35 = Saturate(None=OneMinus(None=MaterialFunctionCall(Hide Element Toggle=@n33, VertexColorChannel=@n4.R, Vertex Value to hide=@n34) {materialFunction=&VertexColorMask}.Result))
  n36 = Constant() {r=0.95}
  n37 = Constant() {r=1}
  n38 = OneMinus(None=Max(A=@n16, B=If(A=@n4.R, B=@n36, "A > B"=@n37, "A == B"=@n19, "A < B"=@n19)))
  n39 = Saturate(None=StaticSwitchParameter(True=Clamp(None=@n31), False=If(A=@n32, B=@n4.R, "A > B"=@n35, "A == B"=@n38, "A < B"=@n38)) {group=Banana, parameterName="OverrideEyebrowLogicWithEye(BuffCat)"})
  n40 = SetMaterialAttributes!(
   MaterialAttributes
   Base Color=NamedRerouteDeclaration(None=@n21.RGB) {name=baseColor, nodeColor=<rgba 0 1 0.1797 1>}
   Metallic=Constant()
   Specular=NamedRerouteDeclaration(None=ScalarParameter() {defaultValue=1, parameterName=Specular}) {name=specular, nodeColor=<rgba 0 1 0.3203 1>}
   Roughness=NamedRerouteDeclaration(None=ScalarParameter() {defaultValue=0.5, parameterName=Roughness}) {name=roughness, nodeColor=<rgba 0.0781 1 0 1>}
   Emissive Color=NamedRerouteDeclaration(None=Multiply(A=@n22, B=@n23)) {name=emissive color, nodeColor=<rgba 0 1 0.9531 1>}
   Opacity Mask=NamedRerouteDeclaration(None=@n21.A) {name=opacityMask, nodeColor=<rgba 0.0547 0 1 1>}
   World Position Offset=NamedRerouteDeclaration(None=Multiply(A=@n24, B=@n39)) {name=WPO, nodeColor=<rgba 1 0 0.7266 1>}
  ) {attributeSetTypes [
    #1
    #2
    #3
    #4
    #5
    #6
    #7
   ]}
  MaterialFunctionCall!(Material=MaterialFunctionCall(In=@n40) {materialFunction=&MF_TechArtOverrides}.Result) {materialFunction=/Game/Packages/Fortress_SharedMaterials/Base_Material_Functions/BaseCharacterShaderFunctions/BaseFN_CharacterOpacity}
  Max(A=@n16, B=Constant())
  n41 = ScalarParameter() {defaultValue=1, parameterName=toggle}
  n42 = VertexColor()
  n43 = ScalarParameter() {defaultValue=0.4}
  n44 = ScalarParameter() {defaultValue=1, parameterName=toggleEyes}
  n45 = ScalarParameter() {defaultValue=0.6, parameterName=ParamEyes}
  Multiply(
   A=@n21.A
   B=Max(
    A=MaterialFunctionCall(Hide Element Toggle=@n41, VertexColorChannel=@n42.G, Vertex Value to hide=@n43) {materialFunction=&VertexColorMask}.Result
    B=MaterialFunctionCall(Hide Element Toggle=@n44, VertexColorChannel=@n42.R, Vertex Value to hide=@n45) {materialFunction=&VertexColorMask}.Result
   )
  )
  Reroute(None=@n4.R)
  Reroute(None=@n16)
  n46 = WorldPosition()
  n47 = Transform(None=@n46.XYZ) {transformSourceType=TRANSFORMSOURCE_World, transformType=TRANSFORM_Camera}
  n48 = ComponentMask(None=@n47) {g=true, r=true}
  n49 = Subtract(A=ComponentMask(None=@n47) {b=true}, B=Add(A=@n42.R, B=@n42.R))
  Subtract(A=Transform(None=AppendVector(A=@n48, B=@n49)) {transformSourceType=TRANSFORMSOURCE_Camera}, B=@n46.XYZ)
 ]
}
